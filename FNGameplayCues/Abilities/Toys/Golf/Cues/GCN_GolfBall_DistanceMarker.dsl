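; Gameplay cue for the golf toy's ball distance marker, tinting the marker's particle effect to match a player's pin color.
; When the start particle system spawns, it takes the instigator from the cue parameters and casts that actor's owner to FortPlayerStateAthena.
; It then writes that player state's pin color to the Cascade component's "Color" parameter.
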(fn OnStartParticleSystemSpawned
  (params (SpawnedParticleSysComponent :type Cascade%20Particle%20System%20Component%20Object%20Reference) (Parameters :type Gameplay%20Cue%20Parameters%20Structure%20%28by%20ref%29))
  (bind (_normalizedmagnitude _rawmagnitude _effectcontext _matchedtagname _originaltag _aggregatedsourcetags _aggregatedtargettags _location _normal _instigator) (Ability|GameplayCue|BreakGameplayCueParameters Parameters))
  (bind _asfort_player_state_athena (Utilities|Casting|CastToFortPlayerStateAthena (Actor|GetOwner _instigator)))
  (Effects|Components|ParticleSystem|SetColorParameter SpawnedParticleSysComponent "Color" (Class|FortPlayerStateZone|GetPinColor _asfort_player_state_athena))
  (return))
